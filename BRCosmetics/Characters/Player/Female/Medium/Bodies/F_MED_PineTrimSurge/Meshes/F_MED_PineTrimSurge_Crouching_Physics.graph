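# Physics asset for the crouching mesh of the PineTrimSurge medium female player body, set up for a two-bone dangling strap.
# Box bodies on dyn_strap_01 and dyn_strap_02 form a constraint chain hanging from a tiny pelvis capsule, with tight twist limit values and wider swing on one axis.
# A right-thigh capsule, angularly locked to the pelvis, supplies a leg volume, apparently for the strap to collide against.
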PhysicsAsset /BRCosmetics/Characters/Player/Female/Medium/Bodies/F_MED_PineTrimSurge/Meshes/F_MED_PineTrimSurge_Crouching_Physics
bodies {
 dyn_strap_01 [{
   shapeType=Box
   extentX=3.3619
   extentY=0.7421
   extentZ=5.488
   shapeName=dyn_strap_01_box
   center=<3.2043 -0.1895 0.085>
   rotation=<rot 0 -90.9731 -88.9701>
  }]
 dyn_strap_02 [{
   shapeType=Box
   extentX=4.391
   extentY=1.0824
   extentZ=9.9072
   shapeName=dyn_strap_02_box
   center=<5.3235 -0.3152 -0.0056>
   rotation=<rot -13.1941 -98.3664 -87.1478>
  }]
 pelvis [{shapeType=Capsule, length=1, radius=1}]
 thigh_r [{shapeType=Capsule, length=22.4173, radius=8.0411, center=<9.8665 2.7196 -0.5277>, rotation=<rot 90 -6.6389 0>}]
}
constraints [
 {
  bone1Name=thigh_r
  bone2Name=pelvis
  swing1LimitDegrees=45
  swing1Motion=Locked
  swing2LimitDegrees=45
  swing2Motion=Locked
  twistLimitDegrees=45
  twistMotion=Locked
 }
 {bone1Name=dyn_strap_01, bone2Name=pelvis, swing1LimitDegrees=12, swing2LimitDegrees=90, twistLimitDegrees=5}
 {bone1Name=dyn_strap_02, bone2Name=dyn_strap_01, swing1LimitDegrees=12, swing2LimitDegrees=45, twistLimitDegrees=5}
]
