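# Post-pose AnimBP for the PineTrimSurge female face accessory mesh, adding secondary motion to the hood and bangs.
# Copies the attached parent's pose, layers Body_Mesh clavicles and upper arms, then runs a parachute-aware Control Rig.
# Chained AnimDynamics springs move hood bones relative to root, with wind on two, plus three bangs bones.
# On customization load, the face material's "Hide Element 02" scalar sets the bangs simulation alpha.
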AnimBlueprint /BRCosmetics/Characters/Player/Female/Medium/Bodies/F_MED_PineTrimSurge/Meshes/Parts/F_MED_PineTrimSurge_FaceAcc_AnimBP
target_skeleton=&/F_MED_PineTrimSurge_FaceAcc_Skeleton
graphs [
 AnimGraph {nodes [
   n1 = Control Rig!(Source=Use cached pose 'Start_Cache'(), Alpha, ParachuteOpened:bool=Get bIsParachuteOpened()) -> (Pose) {
    alphaBoolBlend {blendOption=Linear}
    controlRigAssetReference {blueprintRigClass=&/F_MED_PineTrimSurge_FaceAcc_CtrlRig.F_MED_PineTrimSurge_FaceAcc_CtrlRig_C}
   }
   n2 = Anim Dynamics - Bone: dyn_hood_l(ComponentPose=Local To Component(LocalPose=@n1), Alpha=1, ExternalForce="0,0,0") {
    angularDampingOverride=0.999
    angularSpringConstant=350
    bAngularSpring=true
    bEnableWind=true
    bOverrideAngularDamping=true
    bUseGravityOverride=true
    gravityScale=0.35
    simulationSpace=RootRelative
    windScale=0.4
    alphaBoolBlend {blendOption=Linear}
    boundBone {boneName=dyn_hood_l}
    componentAppliedLinearAccClamp=<3000 3000 3000>
    componentLinearAccScale=<1 1 1>
    componentLinearVelScale=<1 1 1>
    gravityOverride {z=-980}
    physicsBodyDefinitions [{
      boundBone {boneName=dyn_hood_l}
      boxExtents=<15 12 12>
      constraintSetup {angularLimitsMax=<25 60 30>, angularLimitsMin=<-25 -60 -30>, angularTarget=<1 0 0>}
      localJointOffset=<5 0 0>
     }]
    relativeSpaceBone {boneName=head}
   }
   n3 = Anim Dynamics - Bone: dyn_hood_left_02(ComponentPose=@n2, Alpha=1, ExternalForce="0,0,0") {
    angularDampingOverride=0.999
    angularSpringConstant=200
    bAngularSpring=true
    bOverrideAngularDamping=true
    bUseGravityOverride=true
    gravityScale=0.35
    simulationSpace=RootRelative
    alphaBoolBlend {blendOption=Linear}
    boundBone {boneName=dyn_hood_left_02}
    componentAppliedLinearAccClamp=<3000 3000 3000>
    componentLinearAccScale=<0.7 0.7 0.7>
    componentLinearVelScale=<0.7 0.7 0.7>
    gravityOverride {z=-980}
    physicsBodyDefinitions [{
      boundBone {boneName=dyn_hood_left_02}
      boxExtents=<15 12 12>
      constraintSetup {angularLimitsMax=<15 15 15>, angularLimitsMin=<-15 -15 -15>, angularTarget=<1 0 0>}
      localJointOffset=<2.5 0 0>
     }]
    relativeSpaceBone {boneName=head}
   }
   n4 = Anim Dynamics - Bone: dyn_hood_r(ComponentPose=@n3, Alpha=1, ExternalForce="0,0,0") {
    angularDampingOverride=0.999
    angularSpringConstant=400
    bAngularSpring=true
    bEnableWind=true
    bOverrideAngularDamping=true
    bUseGravityOverride=true
    gravityScale=0.35
    simulationSpace=RootRelative
    windScale=0.7
    alphaBoolBlend {blendOption=Linear}
    boundBone {boneName=dyn_hood_r}
    componentAppliedLinearAccClamp=<3000 3000 3000>
    componentLinearAccScale=<1 1 0.8>
    componentLinearVelScale=<1 1 1>
    gravityOverride {z=-980}
    physicsBodyDefinitions [{
      boundBone {boneName=dyn_hood_r}
      boxExtents=<15 12 12>
      constraintSetup {angularLimitsMax=<25 60 30>, angularLimitsMin=<-25 -60 -30>, angularTarget=<1 0 0>}
      localJointOffset=<5 0 0>
     }]
    relativeSpaceBone {boneName=head}
   }
   n5 = Anim Dynamics - Bone: dyn_hood_right_02(ComponentPose=@n4, Alpha=1, ExternalForce="0,0,0") {
    angularDampingOverride=0.999
    angularSpringConstant=250
    bAngularSpring=true
    bOverrideAngularDamping=true
    bUseGravityOverride=true
    gravityScale=0.35
    simulationSpace=RootRelative
    alphaBoolBlend {blendOption=Linear}
    boundBone {boneName=dyn_hood_right_02}
    componentAppliedLinearAccClamp=<3000 3000 3000>
    componentLinearAccScale=<0.7 0.7 0.7>
    componentLinearVelScale=<0.7 0.7 0.7>
    gravityOverride {z=-980}
    physicsBodyDefinitions [{
      boundBone {boneName=dyn_hood_right_02}
      boxExtents=<15 12 12>
      constraintSetup {angularLimitsMax=<15 15 15>, angularLimitsMin=<-15 -15 -15>, angularTarget=<1 0 0>}
      localJointOffset=<2.5 0 0>
     }]
    relativeSpaceBone {boneName=head}
   }
   n6 = Get Alpha() -> (Alpha:float)
   n7 = Anim Dynamics - Bone: dyn_bangs_01(ComponentPose=@n5, Alpha=@n6, ExternalForce="0,0,0") {
    angularDampingOverride=1
    angularSpringConstant=1200
    bAngularSpring=true
    bOverrideAngularDamping=true
    bUseGravityOverride=true
    gravityScale=0.35
    simulationSpace=RootRelative
    alphaBoolBlend {blendOption=Linear}
    boundBone {boneName=dyn_bangs_01}
    componentAppliedLinearAccClamp=<3000 3000 3000>
    componentLinearAccScale=<0.7 0.7 0.7>
    componentLinearVelScale=<0.7 0.7 0.7>
    gravityOverride {z=-980}
    physicsBodyDefinitions [{
      boundBone {boneName=dyn_bangs_01}
      boxExtents=<8 8 8>
      constraintSetup {angularLimitsMax=<15 15 25>, angularLimitsMin=<-15 -15 -2>, angularTarget=<1 0 0>}
      localJointOffset=<4 0 0>
     }]
   }
   n8 = Anim Dynamics - Bone: dyn_bangs_02(ComponentPose=@n7, Alpha=@n6, ExternalForce="0,0,0") {
    angularDampingOverride=1
    angularSpringConstant=1200
    bAngularSpring=true
    bOverrideAngularDamping=true
    bUseGravityOverride=true
    gravityScale=0.35
    simulationSpace=RootRelative
    alphaBoolBlend {blendOption=Linear}
    boundBone {boneName=dyn_bangs_02}
    componentAppliedLinearAccClamp=<3000 3000 3000>
    componentLinearAccScale=<0.7 0.7 0.7>
    componentLinearVelScale=<0.7 0.7 0.7>
    gravityOverride {z=-980}
    physicsBodyDefinitions [{
      boundBone {boneName=dyn_bangs_02}
      boxExtents=<10 10 10>
      constraintSetup {angularLimitsMax=<15 15 25>, angularLimitsMin=<-15 -15 -2>, angularTarget=<1 0 0>}
      localJointOffset=<3.7 0 0>
     }]
   }
   n9 = Anim Dynamics - Bone: dyn_bangs_03(ComponentPose=@n8, Alpha=@n6, ExternalForce="0,0,0") {
    angularDampingOverride=1
    angularSpringConstant=1200
    bAngularSpring=true
    bOverrideAngularDamping=true
    bUseGravityOverride=true
    gravityScale=0.35
    simulationSpace=RootRelative
    alphaBoolBlend {blendOption=Linear}
    boundBone {boneName=dyn_bangs_03}
    componentAppliedLinearAccClamp=<3000 3000 3000>
    componentLinearAccScale=<0.7 0.7 0.7>
    componentLinearVelScale=<0.7 0.7 0.7>
    gravityOverride {z=-980}
    physicsBodyDefinitions [{
      boundBone {boneName=dyn_bangs_03}
      boxExtents=<10 10 10>
      constraintSetup {angularLimitsMax=<15 15 25>, angularLimitsMin=<-15 -15 -2>, angularTarget=<1 0 0>}
      localJointOffset=<3.9 0 0>
     }]
   }
   Output Pose(Result=Component To Local(ComponentPose=@n9)) {layerGroup=None}
   n10 = Layered blend per bone(
    BasePose=Copy Pose From Mesh() {bCopyCurves=true, bUseAttachedParent=true}
    BlendPoses_0=Copy Pose From Mesh(SourceMeshComponent=Get Body_Mesh())
   ) {layerSetup [{branchFilters [{boneName=clavicle_l}, {boneName=clavicle_r}, {boneName=upperarm_l}, {boneName=upperarm_r}]}]}
   Save cached pose 'Start_Cache'(Pose=@n10) {cacheName=Start_Cache}
  ]}
 EventGraph {nodes [
   n1 = Event OnCharacterCustomizationLoaded()
   > Set Body_Mesh(Body_Mesh=@n1.CustomBaseMesh)
   > n2 = Get Dynamic Material Instance(SkeletalMeshComponent=Get Array Item with EFort Custom Part Type(Array=@n1.CustomMeshParts, PartType=Face))
   > n3 = Get Scalar Parameter Value!(execute, self:Material Instance Dynamic Object Reference=@n2, ParameterName:Name=Hide Element 02) -> (then, ReturnValue:float) {is=title K2Node_CallFunction}
   > Set Alpha(Alpha:float=@n3) -> (Output_Get:float)
   Cast To PlayerPawn_Athena(Object=TryGetPawnOwner())
   Event BlueprintUpdateAnimation()
  ]}
]
variables [
 Alpha {type=number}
 Body_Mesh {type=SkeletalMeshComponent}
]
